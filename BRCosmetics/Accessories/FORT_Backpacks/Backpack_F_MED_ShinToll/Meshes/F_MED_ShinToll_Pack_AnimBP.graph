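# Anim Blueprint for the ShinToll backpack cosmetic, adding secondary motion to its ribbons and cables.
# AnimGraph applies small pack and ribbon bone offsets, blends by bool between two ribbon AnimDynamics chains differing in strength, then simulates four cable bones relative to the pack.
# EventGraph updates Gravity Override each frame from a parameter asset, with a fixed downward fallback behind an equality check.
# On customization load, it stores the base mesh and flags whether the body is not Male Large.
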AnimBlueprint /BRCosmetics/Accessories/FORT_Backpacks/Backpack_F_MED_ShinToll/Meshes/F_MED_ShinToll_Pack_AnimBP
target_skeleton=&/F_MED_ShinToll_Pack_Skeleton
graphs [
 AnimGraph {nodes [
   Output Pose(Result=Use cached pose 'DynamicsCache'()) {layerGroup=None}
   n1 = "Transform (Modify) Bone - Bone: pack"!(ComponentPose=Local To Component(LocalPose=Input Pose()), bAlphaBoolEnabled:bool, Translation="0,0.75,0") -> (Pose) {alphaInputType=Bool, translationMode=BMM_Additive, translationSpace=BCS_BoneSpace, boneToModify {boneName=pack}, translation {y=0.75}}
   n2 = "Transform (Modify) Bone - Bone: base_RibbonMain"!(ComponentPose=@n1, Alpha=1, Translation="0,-0.15,0") -> (Pose) {translationMode=BMM_Additive, translationSpace=BCS_BoneSpace, boneToModify {boneName=base_RibbonMain}, translation {y=-0.15}}
   Save cached pose 'CRCache_Input'(Pose=Component To Local(ComponentPose=@n2)) {cacheName=CRCache_Input}
   n3 = Anim Dynamics - Bone: dyn_RibbonMain!(ComponentPose=Local To Component(LocalPose=Use cached pose 'CRCache_Input'()), Alpha=0.6, GravityOverride:Vector="0,0,0") -> (Pose) {
    angularDampingOverride=0.9999
    angularSpringConstant=14000
    bAngularSpring=true
    bOverrideAngularDamping=true
    bUseGravityOverride=true
    gravityScale=0.35
    simulationSpace=BoneRelative
    alphaBoolBlend {blendOption=Linear}
    boundBone {boneName=dyn_RibbonMain}
    componentAppliedLinearAccClamp=<5000 5000 5000>
    componentLinearAccScale=<1 1 1>
    componentLinearVelScale=<1 1 1>
    gravityOverride {z=-980}
    physicsBodyDefinitions [{
      boundBone {boneName=dyn_RibbonMain}
      boxExtents=<35 40 45>
      constraintSetup {angularLimitsMax=<15 5 15>, angularLimitsMin=<-15 -5 -15>, angularTarget=<1 0 0>}
      localJointOffset=<0.5 0 0>
     }]
    relativeSpaceBone {boneName=root}
    simSpaceSettings {maxAngularAcceleration=100, maxAngularVelocity=20, simSpaceAngularAlpha=1}
   }
   n4 = Anim Dynamics - Bone: dyn_Ribbon_Side_l!(ComponentPose=@n3, Alpha=1, GravityOverride:Vector="0,0,0") -> (Pose) {
    angularDampingOverride=0.9999
    angularSpringConstant=14000
    bAngularSpring=true
    bOverrideAngularDamping=true
    bUseGravityOverride=true
    gravityScale=0.35
    simulationSpace=BoneRelative
    alphaBoolBlend {blendOption=Linear}
    boundBone {boneName=dyn_Ribbon_Side_l}
    componentAppliedLinearAccClamp=<5000 5000 5000>
    componentLinearAccScale=<1 1 1>
    componentLinearVelScale=<1 1 1>
    gravityOverride {z=-980}
    physicsBodyDefinitions [{
      boundBone {boneName=dyn_Ribbon_Side_l}
      boxExtents=<35 40 45>
      constraintSetup {angularLimitsMax=<15 15 15>, angularLimitsMin=<-15 -15 -15>, angularTarget=<1 0 0>}
      localJointOffset=<0.25 0 0>
     }]
    relativeSpaceBone {boneName=root}
    simSpaceSettings {maxAngularAcceleration=125, maxAngularVelocity=25, simSpaceAngularAlpha=1}
   }
   n5 = Anim Dynamics - Bone: dyn_Ribbon_Side_r!(ComponentPose=@n4, Alpha=0.85, GravityOverride:Vector="0,0,0") -> (Pose) {
    angularDampingOverride=0.9999
    angularSpringConstant=14000
    bAngularSpring=true
    bOverrideAngularDamping=true
    bUseGravityOverride=true
    gravityScale=0.35
    simulationSpace=BoneRelative
    alphaBoolBlend {blendOption=Linear}
    boundBone {boneName=dyn_Ribbon_Side_r}
    componentAppliedLinearAccClamp=<5000 5000 5000>
    componentLinearAccScale=<1 1 1>
    componentLinearVelScale=<1 1 1>
    gravityOverride {z=-980}
    physicsBodyDefinitions [{
      boundBone {boneName=dyn_Ribbon_Side_r}
      boxExtents=<35 40 45>
      constraintSetup {angularLimitsMax=<15 15 15>, angularLimitsMin=<-15 -15 -15>, angularTarget=<1 0 0>}
      localJointOffset=<0.25 0 0>
     }]
    relativeSpaceBone {boneName=root}
    simSpaceSettings {maxAngularAcceleration=100, maxAngularVelocity=20, simSpaceAngularAlpha=1}
   }
   n6 = Component To Local(ComponentPose=@n5)
   n7 = Anim Dynamics - Bone: dyn_RibbonMain(ComponentPose=Local To Component(LocalPose=Use cached pose 'CRCache_Input'()), Alpha=0.6, GravityOverride="0,0,0") {
    angularDampingOverride=0.9999
    angularSpringConstant=14000
    bAngularSpring=true
    bOverrideAngularDamping=true
    bUseGravityOverride=true
    gravityScale=0.35
    simulationSpace=BoneRelative
    alphaBoolBlend {blendOption=Linear}
    boundBone {boneName=dyn_RibbonMain}
    componentAppliedLinearAccClamp=<5000 5000 5000>
    componentLinearAccScale=<1 1 1>
    componentLinearVelScale=<1 1 1>
    gravityOverride {z=-980}
    physicsBodyDefinitions [{
      boundBone {boneName=dyn_RibbonMain}
      boxExtents=<35 40 45>
      constraintSetup {angularLimitsMax=<15 5 15>, angularLimitsMin=<-15 -5 -15>, angularTarget=<1 0 0>}
      localJointOffset=<0.5 0 0>
     }]
    relativeSpaceBone {boneName=root}
    simSpaceSettings {maxAngularAcceleration=100, maxAngularVelocity=20, simSpaceAngularAlpha=1}
   }
   n8 = Anim Dynamics - Bone: dyn_Ribbon_Side_l(ComponentPose=@n7, Alpha=0.8, GravityOverride="0,0,0") {
    angularDampingOverride=0.9999
    angularSpringConstant=14000
    bAngularSpring=true
    bOverrideAngularDamping=true
    bUseGravityOverride=true
    gravityScale=0.35
    simulationSpace=BoneRelative
    alphaBoolBlend {blendOption=Linear}
    boundBone {boneName=dyn_Ribbon_Side_l}
    componentAppliedLinearAccClamp=<5000 5000 5000>
    componentLinearAccScale=<1 1 1>
    componentLinearVelScale=<1 1 1>
    gravityOverride {z=-980}
    physicsBodyDefinitions [{
      boundBone {boneName=dyn_Ribbon_Side_l}
      boxExtents=<35 40 45>
      constraintSetup {angularLimitsMax=<15 15 15>, angularLimitsMin=<-15 -15 -15>, angularTarget=<1 0 0>}
      localJointOffset=<0.25 0 0>
     }]
    relativeSpaceBone {boneName=root}
    simSpaceSettings {maxAngularAcceleration=125, maxAngularVelocity=25, simSpaceAngularAlpha=1}
   }
   n9 = Anim Dynamics - Bone: dyn_Ribbon_Side_r(ComponentPose=@n8, Alpha=0.6, GravityOverride="0,0,0") {
    angularDampingOverride=0.9999
    angularSpringConstant=14000
    bAngularSpring=true
    bOverrideAngularDamping=true
    bUseGravityOverride=true
    gravityScale=0.35
    simulationSpace=BoneRelative
    alphaBoolBlend {blendOption=Linear}
    boundBone {boneName=dyn_Ribbon_Side_r}
    componentAppliedLinearAccClamp=<5000 5000 5000>
    componentLinearAccScale=<1 1 1>
    componentLinearVelScale=<1 1 1>
    gravityOverride {z=-980}
    physicsBodyDefinitions [{
      boundBone {boneName=dyn_Ribbon_Side_r}
      boxExtents=<35 40 45>
      constraintSetup {angularLimitsMax=<15 15 15>, angularLimitsMin=<-15 -15 -15>, angularTarget=<1 0 0>}
      localJointOffset=<0.25 0 0>
     }]
    relativeSpaceBone {boneName=root}
    simSpaceSettings {maxAngularAcceleration=100, maxAngularVelocity=20, simSpaceAngularAlpha=1}
   }
   n10 = Component To Local(ComponentPose=@n9)
   n11 = Anim Dynamics - Bone: dyn_Cable_Out_01_r!(
    ComponentPose=Local To Component(LocalPose=Blend Poses by bool(BlendPose_0=@n6, BlendPose_1=@n10))
    Alpha=1
    GravityOverride:Vector="0,0,-980"
   ) -> (Pose) {
    angularDampingOverride=1
    angularSpringConstant=5000
    bAngularSpring=true
    bOverrideAngularDamping=true
    bUseGravityOverride=true
    bUsePlanarLimit=false
    gravityScale=0.35
    simulationSpace=BoneRelative
    boundBone {boneName=dyn_Cable_Out_01_r}
    componentAppliedLinearAccClamp=<10000 10000 10000>
    componentLinearAccScale=<1 1 1>
    componentLinearVelScale=<1 1 1>
    gravityOverride {z=-980}
    physicsBodyDefinitions [{
      boundBone {boneName=dyn_Cable_Out_01_r}
      boxExtents=<20 30 30>
      constraintSetup {angularLimitsMax=<25 40 25>, angularLimitsMin=<-25 -60 -45>, angularTarget=<1 0 0>}
      localJointOffset=<3 0 0>
     }]
    relativeSpaceBone {boneName=pack}
    simSpaceSettings {maxAngularAcceleration=8, maxAngularVelocity=3, simSpaceAngularAlpha=1}
   }
   n12 = Anim Dynamics - Bone: dyn_Cable_In_01_r!(ComponentPose=@n11, Alpha=1, GravityOverride:Vector="0,0,-980") -> (Pose) {
    angularDampingOverride=1
    angularSpringConstant=5000
    bAngularSpring=true
    bOverrideAngularDamping=true
    bUseGravityOverride=true
    bUsePlanarLimit=false
    gravityScale=0.35
    simulationSpace=BoneRelative
    boundBone {boneName=dyn_Cable_In_01_r}
    componentAppliedLinearAccClamp=<10000 10000 10000>
    componentLinearAccScale=<1 1 1>
    componentLinearVelScale=<1 1 1>
    gravityOverride {z=-980}
    physicsBodyDefinitions [{
      boundBone {boneName=dyn_Cable_In_01_r}
      boxExtents=<20 30 30>
      constraintSetup {angularLimitsMax=<25 40 25>, angularLimitsMin=<-25 -45 -45>, angularTarget=<1 0 0>}
      localJointOffset=<4 0 0>
     }]
    relativeSpaceBone {boneName=pack}
    simSpaceSettings {maxAngularAcceleration=8, maxAngularVelocity=3, simSpaceAngularAlpha=1}
   }
   n13 = Anim Dynamics - Bone: dyn_Cable_In_01_l!(ComponentPose=@n12, Alpha=1, GravityOverride:Vector="0,0,-980") -> (Pose) {
    angularDampingOverride=1
    angularSpringConstant=5000
    bAngularSpring=true
    bOverrideAngularDamping=true
    bUseGravityOverride=true
    bUsePlanarLimit=false
    gravityScale=0.35
    simulationSpace=BoneRelative
    boundBone {boneName=dyn_Cable_In_01_l}
    componentAppliedLinearAccClamp=<10000 10000 10000>
    componentLinearAccScale=<1 1 1>
    componentLinearVelScale=<1 1 1>
    gravityOverride {z=-980}
    physicsBodyDefinitions [{
      boundBone {boneName=dyn_Cable_In_01_l}
      boxExtents=<20 30 30>
      constraintSetup {angularLimitsMax=<25 45 25>, angularLimitsMin=<-25 -40 -45>, angularTarget=<1 0 0>}
      localJointOffset=<4 0 0>
     }]
    relativeSpaceBone {boneName=pack}
    simSpaceSettings {maxAngularAcceleration=8, maxAngularVelocity=3, simSpaceAngularAlpha=1}
   }
   n14 = Anim Dynamics - Bone: dyn_Cable_Out_01_l!(ComponentPose=@n13, Alpha=1, GravityOverride:Vector="0,0,-980") -> (Pose) {
    angularDampingOverride=1
    angularSpringConstant=5000
    bAngularSpring=true
    bOverrideAngularDamping=true
    bUseGravityOverride=true
    bUsePlanarLimit=false
    gravityScale=0.35
    simulationSpace=BoneRelative
    boundBone {boneName=dyn_Cable_Out_01_l}
    componentAppliedLinearAccClamp=<10000 10000 10000>
    componentLinearAccScale=<1 1 1>
    componentLinearVelScale=<1 1 1>
    gravityOverride {z=-980}
    physicsBodyDefinitions [{
      boundBone {boneName=dyn_Cable_Out_01_l}
      boxExtents=<20 30 30>
      constraintSetup {angularLimitsMax=<25 60 25>, angularLimitsMin=<-25 -30 -45>, angularTarget=<1 0 0>}
      localJointOffset=<3 0 0>
     }]
    relativeSpaceBone {boneName=pack}
    simSpaceSettings {maxAngularAcceleration=8, maxAngularVelocity=3, simSpaceAngularAlpha=1}
   }
   Save cached pose 'DynamicsCache'(Pose=Component To Local(ComponentPose=@n14)) {cacheName=DynamicsCache}
  ]}
 EventGraph {nodes [
   Event BlueprintUpdateAnimation()
   > Sequence()
   > n1 = "Gravity Override State Function (Parameter Asset) (Nativized)"(
    CharacterPartAnimInstance=Self-Reference()
    GameTimeInSeconds=GetGameTimeInSeconds()
    Parameters=&/F_MED_ShinToll_Pack_Parameters
   )
   > n2 = Set Gravity Override(Gravity Override=@n1)
   > Branch(Condition="Equal (Vector)"!(A:Vector=@n2, B:Vector, ErrorTolerance:float) -> (ReturnValue:bool=~) {is=title K2Node_PromotableOperator})
   > Set Gravity Override(Gravity Override="0,0,-400")
   n3 = Event OnCharacterCustomizationLoaded()
   > Set Skel_Mesh(Skel_Mesh=@n3.CustomBaseMesh)
   n4 = Get Array Item with EFort Custom Part Type(Array:Array of Custom Character Part Object References=@n3.CustomParts) -> (Output:Custom Character Part Object Reference)
   n5 = Get GenderPermitted!(self:Custom Character Part Object Reference=@n4) -> (GenderPermitted:EFortCustomGender Enum=Invalid)
   n6 = Get BodyTypesPermitted!(self:Custom Character Part Object Reference=@n4) -> (BodyTypesPermitted:EFortCustomBodyType Enum=NONE)
   > Set isNotMaleLarge(isNotMaleLarge:bool=NAND Boolean(
     A="Equal (Enum)"(A:EFortCustomGender Enum=@n5, B:EFortCustomGender Enum=Male)
     B="Equal (Enum)"(A:EFortCustomBodyType Enum=@n6, B:EFortCustomBodyType Enum=Large)
    ) {is=title K2Node_CommutativeAssociativeBinaryOperator}) -> (Output_Get:bool)
   TryGetPawnOwner()
  ]}
]
variables [
 Gravity Override {type=Vector}
 isNotMaleLarge {type=boolean}
 Skel_Mesh {type=SkeletalMeshComponent}
]
